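# Masked, lit material for a smoky ghost character in the Future Samurai emote, for Niagara mesh particles and skeletal meshes.
# Panning world-aligned storm-cloud noise drives normals, fresnel shading and dithered erosion with particle alpha.
# Vertex offset combines dynamic-parameter bend rotation, noise displacement and stretch toward particle position, with normals corrected for rotation.
# Vertex color masks motion and opacity, and shapes the emissive glow.
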Material /BRCosmetics/Athena/Effects/Particles/Emotes/NS_Emote_FutureSamurai/M_Spooky_Ghost_Character
graph {
 properties {
  bTangentSpaceNormal=false
  bUsedWithMeshDeformer=true
  bUsedWithNiagaraMeshParticles=true
  bUsedWithSkeletalMesh=true
  blendMode=BLEND_Masked
  pixelDepthOffsetMode=PDOM_Legacy
  usedShadingModels=Default Lit
 }
 nodes [
  n1 = SphereMask(
   A=WorldPosition() {worldPositionShaderOffset=WPT_ExcludeAllShaderOffsets}.XYZ
   B=Subtract(
    A=MaterialFunctionCall() {materialFunction=&ObjectPivotPoint}.Object Pivot Location
    B=Constant3Vector() {constant=<rgba 0 0 25 0>}
   )
   Radius=ScalarParameter() {defaultValue=150, parameterName=RotateMaskRadius}
  ) {hardnessPercent=0}
  n2 = Multiply(
   A=LinearInterpolate(
    A=ScalarParameter() {defaultValue=0.7, parameterName=RotateMaskInner}
    B=ScalarParameter() {defaultValue=1, parameterName=RotateMaskOuter}
    Alpha=VertexColor().B
   )
   B=Power(Base=OneMinus(None=@n1), Exp=ScalarParameter() {defaultValue=1, parameterName=RotateMaskPower})
  )
  n3 = Constant3Vector() {constant=<rgba 0 1 0 0>}
  n4 = Constant3Vector() {constant=<rgba 0 0 -1 0>}
  n5 = DynamicParameter() {parameterIndex=1, defaultValue=<rgba -25 0 50 0>, paramNames [X,Y,Z,AxisLerp]}
  n6 = Normalize(VectorInput=Transform(None=LinearInterpolate(A=@n3, B=@n4, Alpha=@n5.AxisLerp)) {transformSourceType=TRANSFORMSOURCE_Local})
  n7 = DynamicParameter() {defaultValue=<rgba 0.5 0 1 1>, paramNames [Bend,Stretch,Param3,Param4]}
  n8 = LinearInterpolate(
   A=ScalarParameter() {defaultValue=0.15, parameterName=RotationAmountMax}
   B=ScalarParameter() {defaultValue=-0.15, parameterName=RotationAmountMin}
   Alpha=@n7.Bend
  )
  n9 = RotateAboutAxis(
   NormalizedRotationAxis=@n6
   RotationAngle=@n8
   PivotPoint=TransformPosition(None=MaterialFunctionCall(R=@n5.X, G=@n5.X, B=@n5.X) {materialFunction=&AppendMany}.RGB) {transformType=TRANSFORMPOSSOURCE_World}
   Position=WorldPosition() {worldPositionShaderOffset=WPT_ExcludeAllShaderOffsets}.XYZ
  )
  n10 = Multiply(A=ScalarParameter() {defaultValue=80, parameterName=Scale}, B=VectorParameter() {parameterName=ScaleMult, defaultValue=<rgba 1 1 1.5 0>}.RGB)
  n11 = Time()
  n12 = ScalarParameter() {defaultValue=50, parameterName=PanSpeed}
  n13 = ActorPositionWS()
  n14 = Add(
   A=Multiply(A=Constant3Vector() {constant=<rgba -0.08 0.08 -1 0>}, B=Multiply(A=@n11, B=@n12))
   B=Subtract(A=WorldPosition().XYZ, B=Multiply(A=@n13) {constB=0.95})
  )
  n15 = Constant()
  n16 = ComponentMask(None=MaterialFunctionCall(
    TextureObject=TextureObject() {samplerType=SAMPLERTYPE_LinearGrayscale, texture=/Game/Global/Textures/Default/Tileable_Patterns/T_StormCloudDetail01_M}
    TextureSize=@n10
    WorldPosition=@n14
    ProjectionTransitionContrast=@n15
   ) {materialFunction=&WorldAlignedTexture}.XY Texture) {g=true}
  n17 = Multiply(A=LinearInterpolate(Alpha=@n16) {constA=-4, constB=5}, B=VertexNormalWS())
  n18 = OneMinus(None=VertexColor().B)
  n19 = Subtract(A=ParticlePositionWS(), B=WorldPosition().XYZ)
  n20 = Normalize(VectorInput=@n19)
  n21 = VertexNormalWS()
  n22 = Multiply(A=@n7.Stretch, B=Saturate(None=DotProduct(A=@n20, B=@n21))) {constB=0.5}
  Add(
   A=Add(A=Multiply(A=@n2, B=@n9), B=Multiply(A=@n17, B=@n18))
   B=LinearInterpolate(B=@n19, Alpha=Power(Base=@n22) {constExponent=5}) {constB=0}
  )
  Constant() {r=1}
  n23 = ScalarParameter() {defaultValue=0.25, parameterName=FresnelMin}
  n24 = ScalarParameter() {defaultValue=2, parameterName=FresnelMax}
  n25 = VertexColor()
  n26 = Fresnel(ExponentIn=Add(A=LinearInterpolate(A=@n23, B=@n24, Alpha=@n16) {constA=0.2, constAlpha=1, constB=1.5}, B=@n25.G)) {baseReflectFraction=0}
  n27 = ScalarParameter() {defaultValue=1, parameterName=FresnelPower}
  n28 = Clamp(None=Multiply(A=LinearInterpolate(Alpha=@n16) {constA=0.5}, B=Power(Base=@n26, Exp=@n27) {constExponent=1.9432}))
  LinearInterpolate(Alpha=@n28) {constA=0.05, constB=0.45}
  n29 = ParticleColor()
  n30 = Add(A=LinearInterpolate(Alpha=@n29.A) {constA=-1}, B=@n16)
  n31 = ScalarParameter() {defaultValue=1, parameterName=OverallOpacity}
  MaterialFunctionCall(Alpha Threshold=Multiply(A=Saturate(None=@n30), B=Multiply(A=@n25.R, B=@n31)) {constA=1}) {materialFunction=&DitherTemporalAA}
  n32 = VertexColor()
  n33 = LinearInterpolate(
   A=VectorParameter() {parameterName=GlowEdge, defaultValue=<rgba 0.1 0.0062 0.0253 0>}.RGB
   B=VectorParameter() {parameterName=Glow, defaultValue=<rgba 1 0.1 0 0>}.RGB
   Alpha=Power(Base=@n32.G) {constExponent=2.5}
  )
  MaterialFunctionCall(Emissive=Multiply(A=Multiply(A=@n33, B=@n32.G), B=@n29.A)) {materialFunction=&MF_FX_EmissiveAdjustments}
  n34 = TextureObject() {samplerType=SAMPLERTYPE_Normal, texture=/Game/Global/Textures/Default/Tileable_Patterns/T_StormCloudDetail01_N}
  n35 = VertexNormalWS()
  n36 = ShadingPathSwitch(
   Default=MaterialFunctionCall(
    Normal=MaterialFunctionCall(TextureObject=@n34, TextureSize=@n10, ProjectionTransitionContrast=@n15, WorldPosition=@n14) {materialFunction=&WorldAlignedNormal}.XY Texture
    Flatness=ScalarParameter() {defaultValue=0.25}
   ) {materialFunction=&FlattenNormal}.Result
   Forward=@n35
   Mobile=@n35
  )
  MaterialFunctionCall(Rotation Axis=@n6, Rotation Angle=@n8, World Space Vertex Normals=Transform(None=QualitySwitch(Default=@n36, Low=@n35)), Movement Mask=@n2) {materialFunction=&FixRotateAboutAxisNormals}
  Multiply(
   A=LinearInterpolate(
    A=VectorParameter() {parameterName=DarkColor, defaultValue=<rgba 0.005 0.0019 0.0024 0>}.RGB
    B=VectorParameter() {parameterName=LightColor, defaultValue=<rgba 0.0455 0.0379 0.05 0>}.RGB
    Alpha=@n28
   )
   B=Add(A=@n28, B=ScalarParameter() {defaultValue=0.1, parameterName=ColorAdd}) {constB=0.05}
  )
  StaticBool() {value=true}
  n37 = Subtract(
   A=ScreenPosition().ViewportUV
   B=MaterialFunctionCall(World Position=ActorPositionWS()) {materialFunction=&TransformToClipSpace}.Buffer UVs for Scene Textures
  )
  n38 = Distance(A=ActorPositionWS(), B=CameraPositionWS())
  n39 = ScalarParameter() {defaultValue=0.05, parameterName=DistanceScale}
  n40 = Multiply(A=Clamp(None=Multiply(A=@n38, B=@n39), Max=ScalarParameter() {defaultValue=500, parameterName=MaxDistanceScale}) {maxDefault=10000}) {constB=0.5}
  TextureSample(UVs=Panner(Coordinate=Multiply(A=@n37, B=@n40)) {speedX=0.05, speedY=0.5}) {samplerType=SAMPLERTYPE_LinearGrayscale, texture=&T_StormCloudDetail01_M}
  VertexNormalWS()
 ]
}
