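# Niagara system for the Future Samurai emote, shrouding the character in inky and red smoke with a glowing eye, lit glow and spinning meshes.
# Decals, SwirlSmoke, RisingSmoke, Eye and Glow follow User.Skeletal Mesh bones or sockets, several indexed by spawn group.
# Shadow meshes, inky smoke, big swirls and spline meshes spawn around fixed heights on staggered loops.
# System.SwitchGlow lerps brightness or size across glow, eye, swirl smoke and splines.
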NiagaraSystem /BRCosmetics/Athena/Effects/Particles/Emotes/NS_Emote_FutureSamurai/NS_Emote_FurureSamurai_Emote
dependencies {
 usedDataInterfaces [
  &NiagaraDataInterfaceCurve
  &NiagaraDataInterfaceArrayFloat
  &NiagaraDataInterfaceVectorCurve
  &NiagaraDataInterfaceVector2DCurve
 ]
 usedDynamicInputs []
 usedModules [
  &self:SystemSpawnScript.SystemScriptSource.SystemScriptGraph.NiagaraNodeAssignment_0.SetVariables_#1
  &SystemState
 ]
}
emitters {
 Decals {
  emitterUpdateScript [
   &EmitterState()
   &SpawnBurst_Instantaneous(Spawn Count=32, Use Spawn Probability=false)
  ]
  particleSpawnScript [
   &FN_SetSpawnGroup_v2(Number of Spawn Groups=32)
   /Niagara/Modules/Spawn/Initialization/InitializeParticle(
    Color=<rgba 0 0 0 5>
    Lifetime=3.75
    Module.Write Color=true
    Module.Write Lifetime=true
    Module.Write Mass=false
    Module.Write Material Random=true
    Module.Write Position=false
    Module.Write Scale=false
    Module.Write SpriteRotation=false
    Module.Write SpriteSize=true
    Position=&SimulationPosition()
    Sprite Size=&Vector2DFromFloat(Value=35)
   )
   &CameraOffset(Camera Offset Amount=25)
  ]
  particleUpdateScript [
   &ParticleState()
   &SkeletalMeshLocation(
    Bone / Socket Index=&ModuloInt(A=$Particles.SpawnGroup, B=16)
    Bone Sampling Mode=::NewEnumerator4
    Inherit Velocity=false
    MaskSpawnGroup=false
    Sampled Position Offset=&MakeCustomVectorFromBool(
     Bool=&SetBoolByIntComparison(A=$Particles.SpawnGroup, B=15, Comparison Type=::NewEnumerator0)
     True Vector=<8 0 0>
     False Vector=<-8 0 0>
    )
    Sampled Position Offset Coordinate Space:ENiagaraMeshTransforms=::NewEnumerator3
    Skeletal Mesh=$User.Skeletal Mesh
   )
   &DynamicMaterialParameters(
    Index 0 Param 1=&FloatFromCurve(
     FloatCurve=DataInterface() {propertyValues {Curve {keys [
         {arriveTangent=-1.656, interpMode=RCIM_Cubic, leaveTangent=-1.656, tangentMode=RCTM_User, time=0, value=1}
         <at 0.85 0 RCIM_Cubic>
         {arriveTangent=8.0037, interpMode=RCIM_Cubic, leaveTangent=8.0037, tangentMode=RCTM_User, time=1, value=1}
        ]}}}
     Scale Curve=3
    )
    Index 0 Param 2=1
    Index 0 Param 4=$Particles.MaterialRandom
    Module.Param0WriteEnabled=true
    Module.Param1WriteEnabled=false
    Module.Param2WriteEnabled=false
    Module.Param3WriteEnabled=false
   )
   &ScaleColor(
    Scale Alpha=&FloatFromCurve(FloatCurve=DataInterface() {propertyValues {Curve {keys [
         {interpMode=RCIM_Cubic, tangentMode=RCTM_User, time=0, value=0}
         {interpMode=RCIM_Cubic, tangentMode=RCTM_User, time=0.1, value=1}
         {interpMode=RCIM_Cubic, tangentMode=RCTM_User, time=0.9, value=1}
         {interpMode=RCIM_Cubic, tangentMode=RCTM_User, time=1, value=0}
        ]}}})
    Scale RGB=&Lerp_Vector(A=<0 0 0>, B=<1 0.1 0.1>, Alpha=$System.SwitchGlow)
    ScaleA=true
    ScaleRGB=false
   )
  ]
  rendererClasses [&NiagaraSpriteRendererProperties]
  renderers [{rendererClass=&NiagaraSpriteRendererProperties}]
 }
 Eye {
  emitterUpdateScript [
   &EmitterState(Life Cycle Mode=::NewEnumerator1, Loop Behavior=::NewEnumerator1, Loop Duration=3, UseLoopDelay=false)
   &SpawnBurst_Instantaneous(Spawn Count=1, Use Spawn Probability=false)
  ]
  particleSpawnScript [
   &InitializeParticle(
    Color=<rgba 1 0.0465 0.0052 1>
    Lifetime=3.5
    Mass=0
    Mesh Scale=&VectorFromFloat(Value=0.25)
    Module.Write Color=true
    Module.Write Lifetime=true
    Module.Write Mass=true
    Module.Write Material Random=true
    Module.Write Position=true
    Module.Write Scale=true
    Module.Write SpriteRotation=true
    Module.Write SpriteSize=true
    Sprite Size=&Vector2DFromFloat(Value=17)
   )
   &SubUVAnimation(End Frame=0)
   /SetVariables(Particles.SpriteSize2=<30 40>) {isSetParameters=true}
   /SetVariables(Particles.LightExp=3) {isSetParameters=true}
   &DynamicMaterialParameters(
    Index 0 Param 1=1.5
    Index 0 Param 2=0.1
    Module.Param0WriteEnabled=true
    Module.Param1WriteEnabled=true
    Module.Param2WriteEnabled=false
    Module.Param3WriteEnabled=true
   )
   /Niagara/Modules/Spawn/Orientation/V2/InitialMeshOrientation(Rotation=<0 -0.25 0>, Use Orientation Vector=true, Use Rotation Vector=true) {enabled=false}
   /Niagara/Modules/Spawn/Orientation/V3/InitialMeshOrientation(Rotation=<0 -0.25 0>, Use Orientation Vector=true, Use Rotation Vector=true) {module=InitialMeshOrientation001}
  ]
  particleUpdateScript [
   &ParticleState()
   &SkeletalMeshLocation(
    Apply Sampled Orientation to Mesh Particles=true
    Bone Sampling Mode=::NewEnumerator4
    Inherit Velocity=false
    MaskSpawnGroup=false
    Sampled Position Offset=<8 8 0>
    Sampled Position Offset Coordinate Space:ENiagaraMeshTransforms=::NewEnumerator3
    Skeletal Mesh=$User.Skeletal Mesh
   )
   &ScaleColor(
    Scale Alpha=&FloatFromCurve(
     FloatCurve=DataInterface() {propertyValues {Curve {keys [
         <at 0 0 RCIM_Cubic>
         {interpMode=RCIM_Cubic, tangentMode=RCTM_User, time=0.3, value=1}
         {interpMode=RCIM_Cubic, tangentMode=RCTM_User, time=0.7, value=1}
         <at 1 0>
        ]}}}
     Scale Curve=&FloatFromCurve(
      FloatCurve=DataInterface() {propertyValues {Curve {keys [<at 0 0 RCIM_Cubic>,<at 0.3 1>,<at 0.7 1 RCIM_Cubic>,<at 1 0>]}}}
      CurveIndex=$System.NormalizedLoopAge
     )
    )
    Scale RGB=&VectorFromFloat(Value=&Lerp_Float(A=1, B=5, Alpha=$System.SwitchGlow))
    ScaleA=true
    ScaleRGB=true
   )
   /SetVariables(Particles.LightColor=&MultiplyLinearColor(
     Linear Color=$Particles.Color
     Scale Factor=&MakeVector4_FromFloat(Float=&Multiply_Float(
       A=&MakeFloatFromLinearColor(LinearColor=$Particles.Color, Channel:ENiagaraLinearColor_Channels=::NewEnumerator3)
       B=10
      ))
    )) {isSetParameters=true, module=SetVariables_#2}
   /SetVariables(Particles.DarkenColor:LinearColor=&MultiplyLinearColor(Linear Color=<rgba 0 0 0 2>, Scale Factor=&MakeVector4_FromLinearColor(LinearColor=$Particles.Color))) {isSetParameters=true}
   /NiagaraFN/Sprite/FN_SpriteAlignmentAndFacing(SpriteAlignment:Vector3f=&RotateVectorByQuaternion(Vector To Rotate=<0 1 0>, Quaternion=$Output.SkeletalMeshLocation.SampledRotation))
  ]
  rendererClasses [&NiagaraSpriteRendererProperties, &NiagaraLightRendererProperties]
  renderers [
   {rendererClass=&NiagaraSpriteRendererProperties}
   {rendererClass=&NiagaraLightRendererProperties, rendererIndex=1}
   {rendererClass=&NiagaraSpriteRendererProperties, rendererIndex=2}
  ]
 }
 Glow {
  emitterUpdateScript [
   &EmitterState(Life Cycle Mode=::NewEnumerator1, Loop Behavior=::NewEnumerator1, Loop Delay=0.5, Loop Duration=3, UseLoopDelay=true)
   &SpawnBurst_Instantaneous(Spawn Count=1, Use Spawn Probability=false)
  ]
  particleSpawnScript [
   &InitializeParticle(
    Color=<rgba 1 0.03 0.02 1>
    Lifetime=3.4
    Mass=0
    Module.Write Color=true
    Module.Write Lifetime=true
    Module.Write Mass=true
    Module.Write Material Random=true
    Module.Write Position=true
    Module.Write Scale=false
    Module.Write SpriteRotation=false
    Module.Write SpriteSize=true
    Sprite Size=<150 200>
   )
   &DynamicMaterialParameters(
    Index 0 Param 1=1
    Index 0 Param 2=2
    Module.Param0WriteEnabled=true
    Module.Param1WriteEnabled=true
    Module.Param2WriteEnabled=false
    Module.Param3WriteEnabled=true
   )
   &CameraOffset(Camera Offset Amount=35)
   /SetVariables(Particles.LightExp=5) {isSetParameters=true}
  ]
  particleUpdateScript [
   &ParticleState()
   &SkeletalMeshLocation(
    Bone / Socket Index=2
    Bone Sampling Mode=::NewEnumerator4
    Inherit Velocity=false
    MaskSpawnGroup=false
    Sampled Position Offset=<15 0 0>
    Sampled Position Offset Coordinate Space:ENiagaraMeshTransforms=::NewEnumerator3
    Skeletal Mesh=$User.Skeletal Mesh
   )
   &ScaleColor(
    Scale Alpha=&FloatFromCurve(
     FloatCurve=DataInterface() {propertyValues {Curve {keys [
         <at 0 0 RCIM_Cubic>
         {interpMode=RCIM_Cubic, tangentMode=RCTM_User, time=0.3, value=1}
         {interpMode=RCIM_Cubic, tangentMode=RCTM_User, time=0.7, value=1}
         <at 1 0>
        ]}}}
     Scale Curve=&Lerp_Float(A=1, B=10, Alpha=$System.SwitchGlow)
    )
    ScaleA=true
    ScaleRGB=false
   )
   /SetVariables(Particles.LightColor=&MultiplyLinearColor(
     Linear Color=$Particles.Color
     Scale Factor=&MakeVector4_FromFloat(Float=&Multiply_Float(
       A=&MakeFloatFromLinearColor(LinearColor=$Particles.Color, Channel:ENiagaraLinearColor_Channels=::NewEnumerator3)
       B=50
      ))
    )) {isSetParameters=true, module=SetVariables_#2}
   &FN_SpriteAlignmentAndFacing(SpriteAlignment:Vector3f=&RotateVectorByQuaternion(Vector To Rotate=<0 1 0>, Quaternion=$Output.SkeletalMeshLocation.SampledRotation))
  ]
  rendererClasses [&NiagaraSpriteRendererProperties, &NiagaraLightRendererProperties]
  renderers [
   {rendererClass=&NiagaraSpriteRendererProperties}
   {rendererClass=&NiagaraLightRendererProperties, rendererIndex=1}
  ]
 }
 RisingSmoke {
  emitterUpdateScript [
   &EmitterState(Life Cycle Mode=::NewEnumerator1, Loop Behavior=::NewEnumerator1, Loop Delay=2.75, Loop Duration=1, UseLoopDelay=true)
   &SpawnBurst_Instantaneous(Spawn Count=17, Use Spawn Probability=false)
  ]
  particleSpawnScript [
   &FN_SetSpawnGroup_v2(Number of Spawn Groups=17)
   &InitializeParticle(
    Color=&MultiplyLinearColor(
     Linear Color=&UniformRangedLinearColor(Minimum=<rgba 1 0.0552 0 1>, Maximum=<rgba 3 0.0503 0 1>, Color Channel Mode=::NewEnumerator1)
     Scale Factor=<1 1 1 1>
    )
    Lifetime=&RandomRangeFloat(Minimum=1.4, Maximum=2)
    Module.Write Color=true
    Module.Write Lifetime=true
    Module.Write Mass=false
    Module.Write Material Random=true
    Module.Write Position=false
    Module.Write Scale=false
    Module.Write SpriteRotation=true
    Module.Write SpriteSize=true
    Sprite Size=&UniformRangedVector2D(Minimum=<20 60>, Maximum=<35 90>)
   )
   &SystemLocation(MaskSpawnGroup=false, Offset=<0 0 100>) {enabled=false}
   &CameraOffset(Camera Offset Amount=&UniformAOrBFloat(A=12, B=25))
   &DynamicMaterialParameters(
    Index 0 Param 2=$Particles.MaterialRandom
    Module.Param0WriteEnabled=false
    Module.Param1WriteEnabled=true
    Module.Param2WriteEnabled=false
    Module.Param3WriteEnabled=false
   ) {module=DynamicMaterialParameters001}
  ]
  particleUpdateScript [
   &ParticleState()
   &ScaleColor(
    Scale Alpha=&FloatFromCurve(FloatCurve=DataInterface() {propertyValues {Curve {keys [
         {arriveTangent=-8.0953, interpMode=RCIM_Cubic, leaveTangent=-8.0952, tangentMode=RCTM_User, time=0, value=1}
         {arriveTangent=-0.8371, interpMode=RCIM_Cubic, leaveTangent=-0.8371, time=0.1566, value=0.0567}
         {arriveTangent=1.1185, interpMode=RCIM_Cubic, leaveTangent=1.1185, time=0.8123, value=0.32}
         {interpMode=RCIM_Cubic, leaveTangent=12.0487, time=1, value=1}
        ]}}})
    Scale RGB=&VectorFromFloat(Value=&FloatFromCurve(FloatCurve=DataInterface() {propertyValues {Curve {keys [<at 0 1 RCIM_Cubic>,<at 1 0>]}}}))
    ScaleA=true
    ScaleRGB=true
   )
   &SkeletalMeshLocation(
    Bone / Socket Index=$Particles.SpawnGroup
    Bone Sampling Mode=::NewEnumerator4
    Inherit Velocity=false
    MaskSpawnGroup=false
    Sampled Position Offset=<0 0 -6>
    Skeletal Mesh=$User.Skeletal Mesh
   )
   &DynamicMaterialParameters(
    Index 0 Param 1=&FloatFromCurve(
     FloatCurve=DataInterface() {propertyValues {Curve {keys [
         {arriveTangent=4.5199, interpMode=RCIM_Cubic, leaveTangent=4.5199, tangentMode=RCTM_User, time=0, value=0}
         {arriveTangent=0.8527, interpMode=RCIM_Cubic, leaveTangent=0.8527, tangentMode=RCTM_User, time=0.2951, value=0.6871}
         {arriveTangent=0.3142, interpMode=RCIM_Cubic, leaveTangent=0.3142, tangentMode=RCTM_User, time=1, value=1}
        ]}}}
     Scale Curve=2
    )
    Module.Param0WriteEnabled=true
    Module.Param1WriteEnabled=false
    Module.Param2WriteEnabled=false
    Module.Param3WriteEnabled=false
   )
   &ScaleSpriteSize(Scale Factor=&Vector2DFromCurve(
     Vector2Curve=DataInterface() {propertyValues {XCurve {keys [<at 0 1 RCIM_Cubic>,<at 1 0.6 RCIM_Cubic>]}, YCurve {keys [<at 0 0.3 RCIM_Cubic>,<at 1 1 RCIM_Cubic>]}}}
     ScaleCurve=<1 1>
    ))
  ]
  rendererClasses [&NiagaraSpriteRendererProperties]
  renderers [{rendererClass=&NiagaraSpriteRendererProperties}]
 }
 Shadow_1 {
  emitterUpdateScript [
   &EmitterState(Life Cycle Mode=::NewEnumerator1, Loop Behavior=::NewEnumerator1, Loop Delay=0.25, Loop Duration=2.5, UseLoopDelay=true)
   &SpawnBurst_Instantaneous(Spawn Count=1, Spawn Time=0.05, Use Spawn Probability=false)
   &SpawnRate(SpawnRate=&Divide_Float(A=5.95, B=$Emitter.CurrentLoopDuration), Use Spawn Probability=false)
   &SpawnBurst_Instantaneous(Spawn Count=1, Spawn Time=2.45, Use Spawn Probability=false) {module=SpawnBurst_Instantaneous001}
  ]
  particleSpawnScript [
   &FN_SetSpawnGroup_v2(Number of Spawn Groups=7)
   &InitializeParticle(
    Color=<rgba 0.0208 0.0208 0.0208 1>
    Lifetime=1.2
    Module.Write Color=true
    Module.Write Lifetime=true
    Module.Write Mass=true
    Module.Write Material Random=true
    Module.Write Position=true
    Module.Write Scale=false
    Module.Write SpriteRotation=false
    Module.Write SpriteSize=true
    Position=&MakeVector(
     X=0
     Y=0
     Z=&Add_Float(
      A=110
      B=&SelectFloatFromArray(Direct Array Index=$Particles.SpawnGroup, Float Selection Array=DataInterface() {propertyValues {FloatData [8,18,10,25,20,12,35]}})
     )
    )
    Sprite Size=&Vector2DFromFloat(Value=&RandomRangeFloat(Minimum=80, Maximum=100))
   )
   /Niagara/Modules/Spawn/Orientation/V2/InitialMeshOrientation(
    Rotation=&MakeVector(
     X=0
     Y=0
     Z=&SelectFloatFromArray(Direct Array Index=$Particles.SpawnGroup, Float Selection Array=DataInterface() {propertyValues {FloatData [0.15,0.4,-0.1,0.4,0,0.15,0.25]}})
    )
    Rotation Coordinate Space:ENiagaraMeshTransforms=::NewEnumerator0
    Use Orientation Vector=true
    Use Rotation Vector=true
   ) {enabled=false}
   /Niagara/Modules/Spawn/Orientation/V3/InitialMeshOrientation(
    Rotation=&MakeVector(
     X=0
     Y=0
     Z=&SelectFloatFromArray(
      Direct Array Index=$Particles.SpawnGroup
      Float Selection Array=DataInterface() {propertyValues {FloatData [0.15,0.4,-0.1,0.4,0.5,0.15,0.25]}}
     )
    )
    Rotation Coordinate Space=::NewEnumerator0
    Use Orientation Vector=true
    Use Rotation Vector=true
   ) {module=InitialMeshOrientation001}
   /SetVariables(Particles.RotationRate:NiagaraFloat=&FloatFromCurve(
     FloatCurve=DataInterface() {propertyValues {Curve {keys [<at 0 0.5 RCIM_Cubic>, {arriveTangent=-0.3889, interpMode=RCIM_Cubic, leaveTangent=-0.3889, time=0.2, value=1}, <at 0.9 0.15 RCIM_Cubic>]}}}
     CurveIndex=$Emitter.NormalizedLoopAge
     Scale Curve=0.4
    )) {isSetParameters=true}
  ]
  particleUpdateScript [
   &ParticleState()
   /SetVariables(Particles.Position=&Add_Vector(
     A=Linked() {linkedVariable {name=Particles.Initial.Position, type=NiagaraPosition}}
     B=&MakeVector(
      X=0
      Y=0
      Z=&FloatFromCurve(
       FloatCurve=DataInterface() {propertyValues {Curve {keys [<at 0 0 RCIM_Cubic>, {arriveTangent=-2.1464, interpMode=RCIM_Cubic, leaveTangent=-2.1464, tangentMode=RCTM_User, time=1, value=-1}]}}}
       Scale Curve=17
      )
     )
    )) {isSetParameters=true}
   &ScaleMeshSize(Scale Factor=&VectorFromCurve(VectorCurve=DataInterface() {propertyValues {
       XCurve {keys [
         {arriveTangent=0.9255, interpMode=RCIM_Cubic, leaveTangent=0.3086, tangentMode=RCTM_User, time=0, value=0.7}
         {arriveTangent=-0.1, interpMode=RCIM_Cubic, leaveTangent=-0.1, time=0.6, value=1}
         {arriveTangent=-1.8469, interpMode=RCIM_Cubic, leaveTangent=-1.8469, tangentMode=RCTM_User, time=1, value=0.6}
        ]}
       YCurve {keys [
         {arriveTangent=0.9255, interpMode=RCIM_Cubic, leaveTangent=0.3086, tangentMode=RCTM_User, time=0, value=0.7}
         {arriveTangent=-0.1, interpMode=RCIM_Cubic, leaveTangent=-0.1, time=0.6, value=1}
         {arriveTangent=-1.8469, interpMode=RCIM_Cubic, leaveTangent=-1.8469, tangentMode=RCTM_User, time=1, value=0.6}
        ]}
       ZCurve {keys [
         {arriveTangent=0.7448, interpMode=RCIM_Cubic, leaveTangent=0.7448, tangentMode=RCTM_User, time=0, value=0.67}
         {interpMode=RCIM_Cubic, tangentMode=RCTM_User, time=0.75, value=1}
         {arriveTangent=-0.7773, interpMode=RCIM_Cubic, leaveTangent=-0.7773, tangentMode=RCTM_User, time=1, value=0.9}
        ]}
      }}))
   &ScaleColor(
    Scale Alpha=&FloatFromCurve(FloatCurve=DataInterface() {propertyValues {Curve {keys [
         {arriveTangent=5905.4829, interpMode=RCIM_Cubic, leaveTangent=3.918, tangentMode=RCTM_User, time=0, value=0}
         {arriveTangent=3.3481, interpMode=RCIM_Cubic, leaveTangent=3.3481, tangentMode=RCTM_User, time=0.2461, value=0.4982}
         {interpMode=RCIM_Cubic, tangentMode=RCTM_User, time=0.5, value=1}
         {arriveTangent=-4.2878, interpMode=RCIM_Cubic, leaveTangent=-4.2878, tangentMode=RCTM_User, time=0.7484, value=0.4982}
         {arriveTangent=-3.3517, interpMode=RCIM_Cubic, leaveTangent=-3.3517, tangentMode=RCTM_User, time=1, value=0}
        ]}}})
    ScaleA=true
    ScaleRGB=true
   )
   /Niagara/Modules/Update/Orientation/V2/UpdateMeshOrientation(Rotation Rate:NiagaraFloat=$Particles.RotationRate, Rotation Vector:Vector3f=<0 0 1>) {enabled=false}
   /Niagara/Modules/Update/Orientation/V3/UpdateMeshOrientation(Rotation Rate=$Particles.RotationRate, Rotation Vector=<0 0 1>) {module=UpdateMeshOrientation001}
   &DynamicMaterialParameters(
    Index 0 Param 1=&FloatFromCurve(FloatCurve=DataInterface() {propertyValues {Curve {keys [<at 0 0 RCIM_Cubic>,<at 1 1>]}}}, Scale Curve=1.25)
    Index 0 Param 2=&FloatFromCurve(FloatCurve=DataInterface() {propertyValues {Curve {keys [<at 0.12 1 RCIM_Cubic>, {interpMode=RCIM_Cubic, tangentMode=RCTM_User, time=0.4, value=0}, <at 0.75 1 RCIM_Cubic>]}}})
    Index 1 Param 4=$Particles.AxisLerp
    Module.Param0WriteEnabled=true
    Module.Param0WriteEnabled1=false
    Module.Param1WriteEnabled=true
    Module.Param1WriteEnabled1=false
    Module.Param2WriteEnabled=false
    Module.Param2WriteEnabled1=false
    Module.Param3WriteEnabled=false
    Module.Param3WriteEnabled1=true
    Write Parameter Index 1=true
   )
  ]
  rendererClasses [&NiagaraMeshRendererProperties]
  renderers [
   {rendererClass=&NiagaraMeshRendererProperties, rendererIndex=0}
   {rendererClass=&NiagaraMeshRendererProperties, rendererIndex=1}
   {rendererClass=&NiagaraMeshRendererProperties, rendererIndex=2}
   {rendererClass=&NiagaraMeshRendererProperties, rendererIndex=3}
   {rendererClass=&NiagaraMeshRendererProperties, rendererIndex=4}
   {rendererClass=&NiagaraMeshRendererProperties, rendererIndex=5}
   {rendererClass=&NiagaraMeshRendererProperties, rendererIndex=6}
   {rendererClass=&NiagaraMeshRendererProperties, rendererIndex=7}
  ]
 }
 SmokeInky {
  emitterUpdateScript [
   &EmitterState(Life Cycle Mode=::NewEnumerator1, Loop Behavior=::NewEnumerator1, Loop Delay=0.2, Loop Duration=3, UseLoopDelay=true)
   &SpawnBurst_Instantaneous(Spawn Count=7, Spawn Time=0.055, Use Spawn Probability=false) {module=SpawnBurst_Instantaneous001}
   &SpawnRate(SpawnRate=6, Use Spawn Probability=false)
   &SpawnBurst_Instantaneous(Spawn Count=8, Spawn Time=2.5, Use Spawn Probability=false)
  ]
  particleSpawnScript [
   &InitializeParticle(
    Color=<rgba 0.0047 0.0048 0.005 1>
    Lifetime=&UniformRangedFloat(Minimum=1.6, Maximum=2.4)
    Module.Write Color=true
    Module.Write Lifetime=true
    Module.Write Mass=true
    Module.Write Material Random=true
    Module.Write Position=false
    Module.Write Scale=false
    Module.Write SpriteRotation=true
    Module.Write SpriteSize=true
    Position=&SimulationPosition()
    Sprite Rotation=&UniformRangedFloat(Minimum=0, Maximum=300)
    Sprite Size=&Vector2DFromFloat(Value=&UniformRangedFloat(Minimum=50, Maximum=100))
   )
   &CylinderLocation(
    Cylinder Height=40
    Cylinder Radius=&RandomRangeFloat(Minimum=10, Maximum=30)
    MaskSpawnGroup=false
    Module.Enable Lathe Profile=false
    Module.Override Randomness=false
    Module.Surface Only=false
    Offset=&MakeVector(
     X=0
     Y=0
     Z=&FloatFromCurve(
      FloatCurve=DataInterface() {propertyValues {Curve {keys [<at 0 0.9 RCIM_Cubic>, {arriveTangent=0.75, interpMode=RCIM_Cubic, leaveTangent=0.75, time=0.6354, value=1.2051}, <at 0.8 1.5 RCIM_Cubic>]}}}
      CurveIndex=$Emitter.NormalizedLoopAge
      Scale Curve=120
     )
    )
   )
   &CameraOffset(Camera Offset Amount=55)
  ]
  particleUpdateScript [
   &ParticleState()
   &GravityForce(Gravity=<0 0 -30>)
   &ScaleColor(
    Scale Alpha=&FloatFromCurve(
     FloatCurve=DataInterface() {propertyValues {Curve {keys [
         {arriveTangent=6.1149, interpMode=RCIM_Cubic, leaveTangent=6.1149, tangentMode=RCTM_User, time=0, value=0}
         {interpMode=RCIM_Cubic, tangentMode=RCTM_User, time=0.2, value=1}
         {arriveTangent=-1.2458, interpMode=RCIM_Cubic, leaveTangent=-1.2458, time=0.7743, value=0.8001}
         {interpMode=RCIM_Cubic, leaveTangent=-4.41, time=1.0011, value=0.002}
        ]}}}
     Scale Curve=&RandomRangeFloat(Minimum=4, Maximum=5)
    )
    ScaleA=true
    ScaleRGB=false
   )
   &SolveForcesAndVelocity(Clamp Velocity=false, Limit Acceleration=false)
   &ScaleSpriteSize(Scale Factor=&Vector2DFromFloat(Value=&FloatFromCurve(FloatCurve=DataInterface() {propertyValues {Curve {keys [
          {arriveTangent=6.3444, interpMode=RCIM_Cubic, time=0, value=0.25}
          {arriveTangent=1.3228, interpMode=RCIM_Cubic, leaveTangent=1.3228, tangentMode=RCTM_User, time=0.2915, value=0.6022}
          {interpMode=RCIM_Cubic, tangentMode=RCTM_User, time=0.8, value=1}
          {arriveTangent=-1.768, interpMode=RCIM_Cubic, leaveTangent=-1.768, tangentMode=RCTM_User, time=1, value=0.8}
         ]}}})))
   &SubUVAnimation(
    Sub UV Play Rate:NiagaraFloat=&RandomRangeFloat(Minimum=0.4, Maximum=0.7)
    SubUV Animation Mode:ENiagara_SubUVLookupMode=::NewEnumerator3
    SubUV Start Time Offset:NiagaraFloat=&RandomRangeFloat(Minimum=0, Maximum=1)
   )
   &DynamicMaterialParameters(
    Index 0 Param 1=&FloatFromCurve(FloatCurve=DataInterface() {propertyValues {Curve {keys [{interpMode=RCIM_Cubic, tangentMode=RCTM_User, time=0.4, value=0}, {interpMode=RCIM_Cubic, tangentMode=RCTM_User, time=1, value=-0.75}]}}})
    Index 0 Param 4=1
    Module.Param0WriteEnabled=true
    Module.Param1WriteEnabled=false
    Module.Param2WriteEnabled=false
    Module.Param3WriteEnabled=true
   )
  ]
  rendererClasses [&NiagaraSpriteRendererProperties]
  renderers [{rendererClass=&NiagaraSpriteRendererProperties}]
 }
 SpinSplines {
  emitterUpdateScript [
   &EmitterState(Life Cycle Mode=::NewEnumerator1, Loop Behavior=::NewEnumerator1, Loop Delay=0.1, Loop Duration=2.8, UseLoopDelay=true)
   &SpawnRate(SpawnRate=8, Use Spawn Probability=false)
   &SpawnBurst_Instantaneous(Spawn Count=8, Spawn Time=2.4, Use Spawn Probability=false) {module=SpawnBurst_Instantaneous001}
  ]
  particleSpawnScript [
   &InitializeParticle(
    Color=<rgba 0.3 0.0087 0 1>
    Lifetime=&RandomRangeFloat(Minimum=1.5, Maximum=2.4)
    Mesh Scale=&RandomRangeVector(Minimum=<0.1 0.15 0.12>, Maximum=<0.4 0.2 0.17>)
    Module.Write Color=true
    Module.Write Lifetime=true
    Module.Write Mass=true
    Module.Write Material Random=true
    Module.Write Position=true
    Module.Write Scale=true
    Module.Write SpriteRotation=false
    Module.Write SpriteSize=false
    Position=&MakeVector(
     X=0
     Y=0
     Z=&FloatFromCurve(
      FloatCurve=DataInterface() {propertyValues {Curve {keys [<at 0 1 RCIM_Cubic>,<at 1 1.5>]}}}
      CurveIndex=$Emitter.NormalizedLoopAge
      Scale Curve=&RandomRangeFloat(Minimum=100, Maximum=125)
     )
    )
    Sprite Size=<5 5>
   )
   /Niagara/Modules/Spawn/Orientation/V2/InitialMeshOrientation(
    Orientation Axis:Vector3f=<0 0 -1>
    Rotation=&MakeVector(
     X=&RandomRangeFloat(Minimum=0, Maximum=1)
     Y=&FloatFromCurve(
      FloatCurve=DataInterface() {propertyValues {Curve {keys [
          <at 0 1 RCIM_Cubic>
          {arriveTangent=-5.2683, interpMode=RCIM_Cubic, leaveTangent=-5.2682, tangentMode=RCTM_User, time=0.4932, value=0}
          <at 1 -1>
         ]}}}
      CurveIndex=&RandomRangeFloat(Minimum=0, Maximum=1)
      Scale Curve=0.08
     )
     Z=0
    )
    Use Orientation Vector=true
    Use Rotation Vector=true
   ) {enabled=false}
   /Niagara/Modules/Spawn/Orientation/V3/InitialMeshOrientation(
    Orientation Axis=<0 0 -1>
    Rotation=&MakeVector(
     X=&RandomRangeFloat(Minimum=0, Maximum=1)
     Y=&FloatFromCurve(
      FloatCurve=DataInterface() {propertyValues {Curve {keys [
          <at 0 1 RCIM_Cubic>
          {arriveTangent=-5.2683, interpMode=RCIM_Cubic, leaveTangent=-5.2682, tangentMode=RCTM_User, time=0.4932, value=0}
          <at 1 -1>
         ]}}}
      CurveIndex=&RandomRangeFloat(Minimum=0, Maximum=1)
      Scale Curve=0.08
     )
     Z=0
    )
    Use Orientation Vector=true
    Use Rotation Vector=true
   ) {module=InitialMeshOrientation001}
  ]
  particleUpdateScript [
   &ParticleState()
   &ScaleColor(
    Scale Alpha=&FloatFromCurve(
     FloatCurve=DataInterface() {propertyValues {Curve {keys [
         <at 0 1 RCIM_Cubic>
         {arriveTangent=1.25, interpMode=RCIM_Cubic, leaveTangent=1.25, time=0.45, value=1.8}
         {interpMode=RCIM_Cubic, tangentMode=RCTM_User, time=0.8, value=2}
         <at 1 0.75 RCIM_Cubic>
        ]}}}
     CurveIndex=$System.NormalizedLoopAge
    )
    Scale RGB=&VectorFromFloat(Value=&FloatFromCurve(
      FloatCurve=DataInterface() {propertyValues {Curve {keys [<at 0 0 RCIM_Cubic>,<at 0.1 1 RCIM_Cubic>,<at 0.75 0 RCIM_Cubic>]}}}
      Scale Curve=&Lerp_Float(A=0.1, B=1, Alpha=$System.SwitchGlow)
     ))
    ScaleA=true
    ScaleRGB=true
   )
   &DynamicMaterialParameters(
    Index 0 Param 1=&FloatFromCurve(FloatCurve=DataInterface() {propertyValues {Curve {keys [
         {arriveTangent=6.1791, interpMode=RCIM_Cubic, leaveTangent=6.1791, tangentMode=RCTM_User, time=0, value=0}
         {interpMode=RCIM_Cubic, tangentMode=RCTM_User, time=0.1943, value=1}
         {arriveTangent=-0.178, interpMode=RCIM_Cubic, leaveTangent=-0.178, tangentMode=RCTM_User, time=0.75, value=0.9}
         <at 1 0 RCIM_Cubic>
        ]}}})
    Index 0 Param 2=&FloatFromCurve(FloatCurve=DataInterface() {propertyValues {Curve {keys [
         {arriveTangent=9.0034, interpMode=RCIM_Cubic, leaveTangent=9.0034, tangentMode=RCTM_User, time=0, value=0}
         {interpMode=RCIM_Cubic, tangentMode=RCTM_User, time=0.2, value=1}
         {interpMode=RCIM_Cubic, tangentMode=RCTM_User, time=0.75, value=1}
         {arriveTangent=-7.1709, interpMode=RCIM_Cubic, leaveTangent=-7.1709, tangentMode=RCTM_User, time=1, value=0}
        ]}}})
    Index 0 Param 4=&RandomRangeFloat(Minimum=0, Maximum=1)
    Module.Param0WriteEnabled=true
    Module.Param1WriteEnabled=true
    Module.Param2WriteEnabled=false
    Module.Param3WriteEnabled=true
   )
   /Niagara/Modules/Update/Orientation/V2/UpdateMeshOrientation(
    Rotation Coordinate Space:ENiagaraMeshTransforms=::NewEnumerator3
    Rotation Rate:NiagaraFloat=&FloatFromCurve(
     FloatCurve=DataInterface() {propertyValues {Curve {keys [
         {arriveTangent=7.0045, interpMode=RCIM_Cubic, leaveTangent=7.0045, tangentMode=RCTM_User, time=0, value=0.2}
         <at 0.25 1 RCIM_Cubic>
         <at 1 0.2 RCIM_Cubic>
        ]}}}
     Scale Curve=&RandomRangeFloat(Minimum=-0.3, Maximum=-0.16)
    )
    Rotation Vector:Vector3f=<1 0 0>
   ) {enabled=false}
   /Niagara/Modules/Update/Orientation/V3/UpdateMeshOrientation(
    Rotation Coordinate Space=::NewEnumerator3
    Rotation Rate=&FloatFromCurve(
     FloatCurve=DataInterface() {propertyValues {Curve {keys [
         {arriveTangent=7.0045, interpMode=RCIM_Cubic, leaveTangent=7.0045, tangentMode=RCTM_User, time=0, value=0.2}
         <at 0.25 1 RCIM_Cubic>
         <at 1 0.2 RCIM_Cubic>
        ]}}}
     Scale Curve=&RandomRangeFloat(Minimum=-0.3, Maximum=-0.16)
    )
    Rotation Vector=<1 0 0>
   ) {module=UpdateMeshOrientation001}
   &ScaleMeshSize(Scale Factor=&VectorFromFloat(Value=&FloatFromCurve(
      FloatCurve=DataInterface() {propertyValues {Curve {keys [
          {arriveTangent=1.85, interpMode=RCIM_Cubic, leaveTangent=1.85, tangentMode=RCTM_User, time=0, value=0.25}
          {arriveTangent=0.1474, interpMode=RCIM_Cubic, leaveTangent=0.1474, tangentMode=RCTM_User, time=1, value=1}
         ]}}}
      Scale Curve=&Lerp_Float(A=1, B=&RandomRangeFloat(Minimum=1.1, Maximum=1.25), Alpha=$System.SwitchGlow)
     )))
   &GravityForce(Gravity=<0 0 -22>)
   &SolveForcesAndVelocity(Clamp Velocity=false, Limit Acceleration=false)
  ]
  rendererClasses [&NiagaraMeshRendererProperties]
  renderers [{rendererClass=&NiagaraMeshRendererProperties, rendererIndex=0}]
 }
 SwirlSmoke {
  emitterUpdateScript [
   &EmitterState(Life Cycle Mode=::NewEnumerator1, Loop Behavior=::NewEnumerator1, Loop Delay=1, Loop Duration=2, UseLoopDelay=true)
   &SpawnRate(SpawnRate=20, Use Spawn Probability=false)
  ]
  particleSpawnScript [
   &FN_SetSpawnGroup_v2(Number of Spawn Groups=32)
   /Niagara/Modules/Spawn/Initialization/V2/InitializeParticle(
    Color=&MakeLinearColorFromVectorAndFloat("Vector (RGB)"=&Lerp_Vector(A=<0.1 0.01 0.01>, B=<2 0.1 0.05>, Alpha=$System.SwitchGlow), "Float (Alpha)"=2)
    Lifetime Max=1.8
    Lifetime Min=1.5
    Lifetime Mode=::NewEnumerator1
    Sprite Rotation Angle=&RandomRangeFloat(Minimum=-15, Maximum=15)
    Sprite Rotation Mode=::NewEnumerator0
    Sprite Size Mode=::NewEnumerator1
    Sprite UV Mode=::NewEnumerator1
    Uniform Sprite Size Max=70
    Uniform Sprite Size Min=50
    UsePositionOffset=false
    Write Material Random=true
   ) {module=InitializeParticle001}
   /SetVariables(Particles.InitialOffset:Vector3f=&UniformRangedVector(Minimum=<-5 -5 -5>, Maximum=<5 5 5>)) {isSetParameters=true}
   &SkeletalMeshLocation(
    Bone / Socket Index=&ModuloInt(A=$Particles.SpawnGroup, B=16)
    Bone Sampling Mode=::NewEnumerator4
    Inherit Velocity=false
    MaskSpawnGroup=false
    Sampled Position Offset=&Add_Vector(
     A=$Particles.InitialOffset
     B=&MakeCustomVectorFromBool(
      Bool=&SetBoolByIntComparison(A=$Particles.SpawnGroup, B=16, Comparison Type=::NewEnumerator0)
      True Vector=<8 0 0>
      False Vector=<-8 0 0>
     )
    )
    Sampled Position Offset Coordinate Space:ENiagaraMeshTransforms=::NewEnumerator3
    Skeletal Mesh=$User.Skeletal Mesh
   ) {module=SkeletalMeshLocation001}
   &CameraOffset(Camera Offset Amount=&UniformRangedFloat(Minimum=5, Maximum=30))
  ]
  particleUpdateScript [
   &ParticleState()
   &GravityForce(Gravity=<0 0 -25>)
   &SolveForcesAndVelocity(Clamp Velocity=false, Limit Acceleration=false)
   &ScaleColor(
    Scale Alpha=&FloatFromCurve(FloatCurve=DataInterface() {propertyValues {Curve {keys [
         {arriveTangent=3.8406, interpMode=RCIM_Cubic, leaveTangent=3.8406, tangentMode=RCTM_User, time=0, value=0}
         {arriveTangent=0.8, interpMode=RCIM_Cubic, leaveTangent=0.8, time=0.859, value=1.5}
         {arriveTangent=-0.9204, interpMode=RCIM_Cubic, leaveTangent=-0.9204, tangentMode=RCTM_User, time=1, value=0.8}
        ]}}})
    Scale RGB=&VectorFromFloat(Value=&Lerp_Float(A=0.7, B=5, Alpha=$System.SwitchGlow))
    ScaleA=true
    ScaleRGB=true
   )
   &SubUVAnimation(
    End Frame:NiagaraInt32=&RandomRangeInt(Minimum=60, Maximum=63)
    Start Frame:NiagaraInt32=&RandomRangeInt(Minimum=12, Maximum=17)
   )
   &DynamicMaterialParameters(
    Index 0 Param 3=&FloatFromCurve(FloatCurve=DataInterface() {propertyValues {Curve {keys [<at 0 0 RCIM_Cubic>,<at 0.5 1>]}}}, CurveIndex=$Particles.NormalizedLife)
    Module.Param0WriteEnabled=false
    Module.Param1WriteEnabled=false
    Module.Param2WriteEnabled=true
    Module.Param3WriteEnabled=false
   )
   &ScaleSpriteSize(Scale Factor=&Vector2DFromFloat(Value=&FloatFromCurve(
      FloatCurve=DataInterface() {propertyValues {Curve {keys [
          {arriveTangent=6.3444, interpMode=RCIM_Cubic, time=0, value=0.2}
          {arriveTangent=1.85, interpMode=RCIM_Cubic, leaveTangent=1.85, tangentMode=RCTM_User, time=0.2598, value=0.5651}
          {interpMode=RCIM_Cubic, tangentMode=RCTM_User, time=0.8, value=1}
          {arriveTangent=-1.768, interpMode=RCIM_Cubic, leaveTangent=-1.768, tangentMode=RCTM_User, time=1, value=0.8}
         ]}}}
      Scale Curve=&Lerp_Float(A=1, B=1.5, Alpha=$System.SwitchGlow)
     )))
  ]
  rendererClasses [&NiagaraSpriteRendererProperties]
  renderers [{rendererClass=&NiagaraSpriteRendererProperties}]
 }
 SwirlSmokeBig {
  emitterUpdateScript [
   &EmitterState(Life Cycle Mode=::NewEnumerator1, Loop Behavior=::NewEnumerator1, Loop Delay=2, Loop Duration=1.1, UseLoopDelay=true)
   &SpawnBurst_Instantaneous(Spawn Count=8, Spawn Time=0.3, Use Spawn Probability=false) {module=SpawnBurst_Instantaneous001}
   &SpawnRate(SpawnRate=10, Use Spawn Probability=false)
  ]
  particleSpawnScript [
   &InitializeParticle(
    Color=<rgba 1 0.0312 0.0312 1.25>
    Lifetime Max=3
    Lifetime Min=2
    Lifetime Mode=::NewEnumerator1
    Sprite Rotation Angle=&RandomRangeFloat(Minimum=-15, Maximum=15)
    Sprite Rotation Mode=::NewEnumerator0
    Sprite Size Mode=::NewEnumerator1
    Uniform Sprite Size Max=100
    Uniform Sprite Size Min=60
    UsePositionOffset=false
    Write Material Random=true
   ) {module=InitializeParticle001}
   &CylinderLocation(
    Cylinder Height=70
    Cylinder Radius=&RandomRangeFloat(Minimum=15, Maximum=30)
    MaskSpawnGroup=false
    Module.Enable Lathe Profile=false
    Module.Override Randomness=false
    Module.Surface Only=false
    Offset=<0 0 140>
   )
   &CameraOffset(Camera Offset Amount=50)
  ]
  particleUpdateScript [
   &ParticleState()
   &ScaleColor(
    Scale Alpha=&FloatFromCurve(FloatCurve=DataInterface() {propertyValues {Curve {keys [
         <at 0 0 RCIM_Cubic>
         {arriveTangent=0.002, interpMode=RCIM_Cubic, leaveTangent=0.002, time=0.3, value=1}
         {arriveTangent=-4.4318, interpMode=RCIM_Cubic, leaveTangent=-4.4318, tangentMode=RCTM_User, time=1.0011, value=0.002}
        ]}}})
    Scale RGB=&VectorFromFloat(Value=&Lerp_Float(A=0.15, B=1.5, Alpha=$System.SwitchGlow))
    ScaleA=true
    ScaleRGB=true
   )
   &GravityForce(Gravity=&RandomRangeVector(Minimum=<0 0 -40>, Maximum=<0 0 -60>))
   &SolveForcesAndVelocity(Clamp Velocity=false, Limit Acceleration=false)
   &ScaleSpriteSize(Scale Factor=&Vector2DFromFloat(Value=&FloatFromCurve(
      FloatCurve=DataInterface() {propertyValues {Curve {keys [
          {arriveTangent=6.3444, interpMode=RCIM_Cubic, time=0, value=0.25}
          {arriveTangent=1.3228, interpMode=RCIM_Cubic, leaveTangent=1.3228, tangentMode=RCTM_User, time=0.2915, value=0.6022}
          {interpMode=RCIM_Cubic, tangentMode=RCTM_User, time=0.8, value=1}
          {arriveTangent=-1.768, interpMode=RCIM_Cubic, leaveTangent=-1.768, tangentMode=RCTM_User, time=1, value=0.7}
         ]}}}
      Scale Curve=&Lerp_Float(A=1, B=1.5, Alpha=$System.SwitchGlow)
     ))) {module=ScaleSpriteSize001}
   &DynamicMaterialParameters(
    Index 0 Param 3=&FloatFromCurve(FloatCurve=DataInterface() {propertyValues {Curve {keys [<at 0 0 RCIM_Cubic>,<at 0.5 1>]}}}, CurveIndex=$Particles.NormalizedLife)
    Module.Param0WriteEnabled=false
    Module.Param1WriteEnabled=false
    Module.Param2WriteEnabled=true
    Module.Param3WriteEnabled=false
   )
   &SubUVAnimation(
    End Frame:NiagaraInt32=&RandomRangeInt(Minimum=56, Maximum=62)
    Start Frame:NiagaraInt32=&RandomRangeInt(Minimum=14, Maximum=22)
   )
  ]
  rendererClasses [&NiagaraSpriteRendererProperties]
  renderers [{rendererClass=&NiagaraSpriteRendererProperties}]
 }
}
summary {
 systemName=NS_Emote_FurureSamurai_Emote
 emitters [
  {emitterName=SwirlSmokeBig, rendererClasses [&NiagaraSpriteRendererProperties]}
  {emitterName=Shadow_1, rendererClasses [&NiagaraMeshRendererProperties]}
  {emitterName=SmokeInky, rendererClasses [&NiagaraSpriteRendererProperties]}
  {emitterName=SwirlSmoke, rendererClasses [&NiagaraSpriteRendererProperties]}
  {emitterName=RisingSmoke, rendererClasses [&NiagaraSpriteRendererProperties]}
  {emitterName=Decals, rendererClasses [&NiagaraSpriteRendererProperties]}
  {emitterName=Eye, rendererClasses [&NiagaraSpriteRendererProperties, &NiagaraLightRendererProperties]}
  {emitterName=Glow, rendererClasses [&NiagaraSpriteRendererProperties, &NiagaraLightRendererProperties]}
  {emitterName=SpinSplines, rendererClasses [&NiagaraMeshRendererProperties]}
 ]
 userVariables [User.Skeletal Mesh {
   type=NiagaraDataInterfaceSkeletalMesh
   defaultValue {
    struct=NiagaraExt_VariableValue_DataInterface
    value {dataInterface=&self:NiagaraDataInterfaceSkeletalMesh_1}
   }
  }]
}
