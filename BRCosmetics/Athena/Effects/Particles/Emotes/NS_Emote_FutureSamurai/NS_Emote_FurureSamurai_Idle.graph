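# Niagara idle effect for the Future Samurai emote, wrapping the character's head area in dark, red-tinted, inky smoke.
# A single Darken sprite underlies two trailing smoke emitters (SmokeInky, SwirlSmoke) and a slower SwirlSmokeLocal layer.
# System head-speed and spawn-rate values drive trail velocity and emission; trails fade and die with distance from System.Offset, tightening near the camera.
# User.bisHeadHidden sets every particle's VisibilityTag.
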NiagaraSystem /BRCosmetics/Athena/Effects/Particles/Emotes/NS_Emote_FutureSamurai/NS_Emote_FurureSamurai_Idle
dependencies {
 usedDataInterfaces [&NiagaraDataInterfaceCurve]
 usedDynamicInputs [
  /Niagara/DynamicInputs/Transforms/TransformPosition
  &self:MaxSpeedNormalized
  &MakeCustomFloatFromBool
  &VectorLength
 ]
 usedModules [
  &SystemState
  &self:SystemSpawnScript.SystemScriptSource.SystemScriptGraph.NiagaraNodeAssignment_2.SetVariables_#1
  &self:SystemSpawnScript.SystemScriptSource.SystemScriptGraph.NiagaraNodeAssignment_0.SetVariables_#2
  &self:SystemSpawnScript.SystemScriptSource.SystemScriptGraph.NiagaraNodeAssignment_1.SetVariables_#3
 ]
}
emitters {
 Darken {
  emitterUpdateScript [
   &EmitterState(
    Inactive Response=::NewEnumerator1
    Life Cycle Mode=::NewEnumerator1
    Loop Behavior=::NewEnumerator1
    Loop Duration Mode=::NewEnumerator1
    UseLoopDelay=false
   )
   &SpawnBurst_Instantaneous(Spawn Count=1, Use Spawn Probability=false)
  ]
  particleSpawnScript [
   /Niagara/Modules/Spawn/Initialization/InitializeParticle(
    Color=<rgba 0 0 0 1.5>
    Lifetime=3.5
    Mass=0
    Mesh Scale=&VectorFromFloat(Value=0.25)
    Module.Write Color=true
    Module.Write Lifetime=true
    Module.Write Mass=true
    Module.Write Material Random=true
    Module.Write Position=true
    Module.Write Scale=true
    Module.Write SpriteRotation=true
    Module.Write SpriteSize=true
    Position=<10 4 0>
    Sprite Size=&Vector2DFromFloat(Value=24)
   )
   &DynamicMaterialParameters(
    Index 0 Param 1=1.5
    Index 0 Param 2=1
    Module.Param0WriteEnabled=true
    Module.Param1WriteEnabled=true
    Module.Param2WriteEnabled=false
    Module.Param3WriteEnabled=false
   )
  ]
  particleUpdateScript [
   &ParticleState(Kill Particles When Lifetime Has Elapsed=false)
   /SetVariables(Particles.VisibilityTag=&MakeCustomIntFromBool(Bool=$User.bisHeadHidden, True INT=1, False INT=0)) {isSetParameters=true, module=SetVariables_#4}
  ]
  rendererClasses [&NiagaraSpriteRendererProperties]
  renderers [{rendererClass=&NiagaraSpriteRendererProperties}]
 }
 SmokeInky {
  emitterUpdateScript [
   &EmitterState()
   /SetVariables(Emitter.NearCamShrink:NiagaraFloat=&self:CameraShrinkThreshold(DistanceOffset:NiagaraFloat=-70, DistanceThreshold:NiagaraFloat=180)) {isSetParameters=true}
   &SpawnRate(SpawnRate=&Multiply_Float(A=8, B=$System.SpawnRateMult), Use Spawn Probability=false)
   &SpawnPerUnit(
    Max Movement Threshold=2000
    Module.Use Max Movement Threshold=true
    Module.Use Movement Tolerance=false
    Spawn Per Unit=10
    Spawn Probability:NiagaraFloat=$System.SpawnRateMult
    Use Spawn Probability=true
   )
  ]
  particleSpawnScript [
   &InitializeParticle(
    Color=<rgba 0.0047 0.0048 0.005 4>
    Lifetime=&UniformRangedFloat(Minimum=0.6, Maximum=0.9)
    Module.Write Color=true
    Module.Write Lifetime=true
    Module.Write Mass=true
    Module.Write Material Random=true
    Module.Write Position=false
    Module.Write Scale=false
    Module.Write SpriteRotation=true
    Module.Write SpriteSize=true
    Position=&SimulationPosition()
    Sprite Rotation=&UniformRangedFloat(Minimum=0, Maximum=1)
    Sprite Size=&Vector2DFromFloat(Value=&UniformRangedFloat(Minimum=35, Maximum=42))
   )
   &CameraOffset(Camera Offset Amount=&Lerp_Float(A=-7, B=6, Alpha=$Emitter.NearCamShrink))
   &SphereLocation(
    MaskSpawnGroup=false
    Non Uniform Scale=<0.1 1 1>
    Offset=<10 4 0>
    Override Seed=false
    Sphere Radius=5
    Surface Only=false
    UseNonUniformScale=true
    UseOffset=true
    UseSphereOrientation=false
   )
   &InheritVelocity(Module.Limit Speed=true, Velocity Limit=500, Velocity Scale=&VectorFromFloat(Value=0.7))
   &AddVelocity(Velocity=&self:SideVelocityWhenMovingForward(
     Multiply:NiagaraFloat=$System.HeadSpeed
     SideVelocity:NiagaraFloat=&UniformAOrBFloat(A=-400, B=400)
     XY Velocity:Vector2f=<300 150>
    )) {module=AddVelocity002}
   &AddVelocity(Velocity=&RandomRangeVector(Minimum=<-2 20 0>, Maximum=<4 25 0>))
  ]
  particleUpdateScript [
   &ParticleState()
   /SetVariables(Particles.DistanceFromOrigin:NiagaraFloat=&FloatFromCurve(
     FloatCurve=DataInterface() {propertyValues {Curve {keys [
         <at 0 0 RCIM_Cubic>
         {arriveTangent=1.7524, interpMode=RCIM_Cubic, leaveTangent=1.7524, tangentMode=RCTM_User, time=0.2, value=0.35}
         <at 1 1>
        ]}}}
     CurveIndex=&Divide_Float(
      A=/Niagara/DynamicInputs/TypeConversions/VectorLengthSafe(Vector:Vector3f=&Subtract_Vector(A=$System.Offset, B=Linked() {linkedVariable {name=Particles.Position, type=NiagaraPosition}}))
      B=&Lerp_Float(A=25, B=80, Alpha=$Emitter.NearCamShrink)
     )
    )) {isSetParameters=true}
   &KillParticles(Kill Particles=&SetBoolByFloatComparison(A=$Particles.DistanceFromOrigin, B=1, Comparison Type=::NewEnumerator2))
   &GravityForce(Gravity=<0 0 -17>) {enabled=false}
   &Drag(Drag=6, Use Linear Drag=true, Use Rotational Drag=false)
   &ScaleColor(
    Scale Alpha=&FloatFromCurve(
     FloatCurve=DataInterface() {propertyValues {Curve {keys [
         {arriveTangent=12.8513, interpMode=RCIM_Cubic, leaveTangent=12.8513, tangentMode=RCTM_User, time=0, value=0}
         {arriveTangent=-1.6464, interpMode=RCIM_Cubic, leaveTangent=-1.6464, tangentMode=RCTM_User, time=0.2, value=1}
         {interpMode=RCIM_Cubic, leaveTangent=-4.41, time=1.0011, value=0.002}
        ]}}}
     Scale Curve=&Lerp_Float(A=1, B=0, Alpha=$Particles.DistanceFromOrigin)
    )
    ScaleA=true
    ScaleRGB=false
   )
   &SolveForcesAndVelocity(Clamp Velocity=false, Limit Acceleration=false)
   &ScaleSpriteSize(Scale Factor=&Vector2DFromFloat(Value=&FloatFromCurve(
      FloatCurve=DataInterface() {propertyValues {Curve {keys [
          {arriveTangent=9.3216, interpMode=RCIM_Cubic, leaveTangent=1.9791, tangentMode=RCTM_User, time=0, value=0.5}
          {arriveTangent=0.5737, interpMode=RCIM_Cubic, leaveTangent=0.5737, tangentMode=RCTM_User, time=0.2709, value=0.82}
          {arriveTangent=0.1252, interpMode=RCIM_Cubic, leaveTangent=0.1252, tangentMode=RCTM_User, time=1, value=1}
         ]}}}
      Scale Curve=&Multiply_Float(A=&Lerp_Float(A=0.6, B=1, Alpha=$Emitter.NearCamShrink), B=1)
     )))
   &SubUVAnimation(
    Sub UV Play Rate:NiagaraFloat=&RandomRangeFloat(Minimum=0.35, Maximum=0.45)
    SubUV Animation Mode:ENiagara_SubUVLookupMode=::NewEnumerator3
    SubUV Start Time Offset:NiagaraFloat=&RandomRangeFloat(Minimum=0, Maximum=1)
   )
   &DynamicMaterialParameters(
    Index 0 Param 1=&FloatFromCurve(FloatCurve=DataInterface() {propertyValues {Curve {keys [{interpMode=RCIM_Cubic, tangentMode=RCTM_User, time=0.4, value=0}, {interpMode=RCIM_Cubic, tangentMode=RCTM_User, time=1, value=-0.75}]}}})
    Index 0 Param 4=1
    Module.Param0WriteEnabled=false
    Module.Param1WriteEnabled=false
    Module.Param2WriteEnabled=false
    Module.Param3WriteEnabled=true
   )
   &ScaleSpriteSizeBySpeed(
    Max Scale Factor=<0.3 2>
    Sample Scale Factor By Curve=true
    Scale Factor Curve=DataInterface() {propertyValues {Curve {keys [
        {arriveTangent=0.4496, interpMode=RCIM_Cubic, leaveTangent=0.4496, tangentMode=RCTM_User, time=0, value=0}
        {arriveTangent=0.5947, interpMode=RCIM_Cubic, leaveTangent=0.5947, tangentMode=RCTM_User, time=0.3921, value=0.4248}
        {arriveTangent=1.6095, interpMode=RCIM_Cubic, leaveTangent=1.6095, tangentMode=RCTM_User, time=1, value=1}
       ]}}}
    Velocity Threshold=600
   )
   /SetVariables(Particles.VisibilityTag=&MakeCustomIntFromBool(Bool=$User.bisHeadHidden, True INT=1, False INT=0)) {isSetParameters=true, module=SetVariables_#4}
  ]
  rendererClasses [&NiagaraSpriteRendererProperties]
  renderers [{rendererClass=&NiagaraSpriteRendererProperties}]
 }
 SwirlSmoke {
  emitterUpdateScript [
   &EmitterState()
   /SetVariables(Emitter.NearCamShrink:NiagaraFloat=&self:CameraShrinkThreshold(DistanceOffset:NiagaraFloat=-70, DistanceThreshold:NiagaraFloat=180)) {isSetParameters=true}
   &SpawnRate(SpawnRate=&Multiply_Float(A=7, B=$System.SpawnRateMult), Use Spawn Probability=false)
   &SpawnPerUnit(
    Max Movement Threshold=2000
    Module.Use Max Movement Threshold=true
    Module.Use Movement Tolerance=false
    Spawn Per Unit=10
    Spawn Probability:NiagaraFloat=$System.SpawnRateMult
    Use Spawn Probability=true
   )
  ]
  particleSpawnScript [
   /Niagara/Modules/Spawn/Initialization/V2/InitializeParticle(
    Color=<rgba 0.05 0.005 0.007 1.25>
    Lifetime=&Lerp_Float(
     A=&RandomRangeFloat(Minimum=0.7, Maximum=1.25)
     B=&RandomRangeFloat(Minimum=0.35, Maximum=0.45)
     Alpha=$System.HeadSpeed
    )
    Sprite Size Mode=::NewEnumerator1
    Sprite UV Mode=::NewEnumerator1
    Uniform Sprite Size Max=30
    Uniform Sprite Size Min=22
    UsePositionOffset=false
    Write Material Random=true
   ) {module=InitializeParticle001}
   &SphereLocation(
    MaskSpawnGroup=false
    Non Uniform Scale=<0.1 1 1>
    Offset=<9 4 0>
    Override Seed=false
    Sphere Radius=5
    Surface Only=false
    UseNonUniformScale=true
    UseOffset=true
    UseSphereOrientation=false
   )
   &CameraOffset(Camera Offset Amount=&Lerp_Float(A=-7, B=6, Alpha=$Emitter.NearCamShrink))
   &AddVelocityFromPoint(Module.Use Falloff Distance=false, Origin Offset=<7 -6 0>, Velocity Strength=35)
   &InheritVelocity(Module.Limit Speed=true, Velocity Limit=500, Velocity Scale=&VectorFromFloat(Value=0.6))
   &AddVelocity(Velocity=&self:SideVelocityWhenMovingForward(
     Multiply:NiagaraFloat=$System.HeadSpeed
     SideVelocity:NiagaraFloat=&UniformAOrBFloat(A=-500, B=500)
     XY Velocity:Vector2f=<0 0>
    ))
   /SetVariables(Particles.SubUVOffset:NiagaraInt32=&RandomRangeInt(Minimum=0, Maximum=14)) {isSetParameters=true, module=SetVariables_#5}
  ]
  particleUpdateScript [
   &ParticleState()
   /SetVariables(Particles.DistanceFromOrigin:NiagaraFloat=&FloatFromCurve(
     FloatCurve=DataInterface() {propertyValues {Curve {keys [
         <at 0 0 RCIM_Cubic>
         {arriveTangent=1.7472, interpMode=RCIM_Cubic, leaveTangent=1.7472, tangentMode=RCTM_User, time=0.1501, value=0.434}
         <at 1 1>
        ]}}}
     CurveIndex=&Divide_Float(
      A=&VectorLengthSafe(Vector:Vector3f=&Subtract_Vector(A=$System.Offset, B=Linked() {linkedVariable {name=Particles.Position, type=NiagaraPosition}}))
      B=&Lerp_Float(A=25, B=70, Alpha=$Emitter.NearCamShrink)
     )
    )) {isSetParameters=true}
   &KillParticles(Kill Particles=&SetBoolByFloatComparison(A=$Particles.DistanceFromOrigin, B=1, Comparison Type=::NewEnumerator2))
   &GravityForce(Gravity=<0 0 -24>) {enabled=false}
   &Drag(Drag=5, Use Linear Drag=true, Use Rotational Drag=false)
   &SolveForcesAndVelocity(Clamp Velocity=false, Limit Acceleration=false)
   &ScaleColor(
    Scale Alpha=&FloatFromCurve(
     FloatCurve=DataInterface() {propertyValues {Curve {keys [
         {arriveTangent=25.7032, interpMode=RCIM_Cubic, leaveTangent=25.7033, tangentMode=RCTM_User, time=0, value=0}
         {interpMode=RCIM_Cubic, tangentMode=RCTM_User, time=0.1, value=1}
         {interpMode=RCIM_Cubic, leaveTangent=-2.703, time=1, value=0}
        ]}}}
     Scale Curve=&Lerp_Float(A=1, B=0, Alpha=$Particles.DistanceFromOrigin)
    )
    Scale RGB=&VectorFromFloat(Value=&FloatFromCurve(FloatCurve=DataInterface() {propertyValues {Curve {keys [{arriveTangent=-3.1207, interpMode=RCIM_Cubic, leaveTangent=-3.1207, tangentMode=RCTM_User, time=0, value=1}, <at 0.5 0>]}}}))
    ScaleA=true
    ScaleRGB=true
   )
   &SubUVAnimation(
    End Frame:NiagaraInt32=&Add_Integer(A=$Particles.SubUVOffset, B=30)
    Start Frame:NiagaraInt32=&Add_Integer(A=$Particles.SubUVOffset, B=15)
   )
   &ScaleSpriteSize(Scale Factor=&Vector2DFromFloat(Value=&FloatFromCurve(
      FloatCurve=DataInterface() {propertyValues {Curve {keys [
          {arriveTangent=6.5992, interpMode=RCIM_Cubic, leaveTangent=2.0153, tangentMode=RCTM_User, time=0, value=0.45}
          {arriveTangent=0.671, interpMode=RCIM_Cubic, leaveTangent=0.671, tangentMode=RCTM_User, time=0.2708, value=0.77}
          {arriveTangent=0.2411, interpMode=RCIM_Cubic, leaveTangent=0.2411, tangentMode=RCTM_User, time=1, value=1}
         ]}}}
      Scale Curve=&Multiply_Float(A=&Lerp_Float(A=0.6, B=1, Alpha=$Emitter.NearCamShrink), B=1)
     )))
   &DynamicMaterialParameters(
    Index 0 Param 3=&FloatFromCurve(
     FloatCurve=DataInterface() {propertyValues {Curve {keys [{arriveTangent=4.3778, interpMode=RCIM_Cubic, leaveTangent=4.3778, tangentMode=RCTM_User, time=0, value=0}, <at 0.5 1>]}}}
     CurveIndex=$Particles.NormalizedLife
    )
    Module.Param0WriteEnabled=false
    Module.Param1WriteEnabled=false
    Module.Param2WriteEnabled=true
    Module.Param3WriteEnabled=false
   )
   &ScaleSpriteSizeBySpeed(
    Max Scale Factor=<0.2 3>
    Sample Scale Factor By Curve=true
    Scale Factor Curve=DataInterface() {propertyValues {Curve {keys [
        {arriveTangent=0.4496, interpMode=RCIM_Cubic, leaveTangent=0.4495, tangentMode=RCTM_User, time=0, value=0}
        {arriveTangent=1.6095, interpMode=RCIM_Cubic, leaveTangent=1.6095, tangentMode=RCTM_User, time=1, value=1}
       ]}}}
    Velocity Threshold=600
   )
   /SetVariables(Particles.VisibilityTag=&MakeCustomIntFromBool(Bool=$User.bisHeadHidden, True INT=1, False INT=0)) {isSetParameters=true, module=SetVariables_#4}
  ]
  rendererClasses [&NiagaraSpriteRendererProperties]
  renderers [{rendererClass=&NiagaraSpriteRendererProperties}]
 }
 SwirlSmokeLocal {
  emitterUpdateScript [
   &EmitterState()
   &SpawnRate(SpawnRate=&Multiply_Float(A=4, B=1), Use Spawn Probability=false)
  ]
  particleSpawnScript [
   &InitializeParticle(
    Color=<rgba 0.05 0.005 0.007 1.25>
    Lifetime Max=1.4
    Lifetime Min=1.1
    Lifetime Mode=::NewEnumerator1
    Sprite Rotation Mode=::NewEnumerator2
    Sprite Size Mode=::NewEnumerator1
    Uniform Sprite Size Max=25
    Uniform Sprite Size Min=18
    UsePositionOffset=false
    Write Material Random=true
   ) {module=InitializeParticle001}
   &SphereLocation(
    MaskSpawnGroup=false
    Non Uniform Scale=<0.1 1 1>
    Offset=<10 8 0>
    Override Seed=false
    Sphere Radius=2
    Surface Only=false
    UseNonUniformScale=true
    UseOffset=true
    UseSphereOrientation=false
   )
   /SetVariables(Particles.SubUVOffset:NiagaraInt32=&RandomRangeInt(Minimum=0, Maximum=15)) {isSetParameters=true, module=SetVariables_#5}
  ]
  particleUpdateScript [
   &ParticleState()
   &ScaleColor(
    Scale Alpha=&FloatFromCurve(FloatCurve=DataInterface() {propertyValues {Curve {keys [
         {arriveTangent=25.7032, interpMode=RCIM_Cubic, leaveTangent=25.7033, tangentMode=RCTM_User, time=0, value=0}
         {interpMode=RCIM_Cubic, tangentMode=RCTM_User, time=0.1, value=1}
         {arriveTangent=-1.1111, interpMode=RCIM_Cubic, leaveTangent=-1.1111, time=0.5929, value=0.8568}
         {interpMode=RCIM_Cubic, leaveTangent=-2.703, time=1, value=0}
        ]}}})
    Scale RGB=&VectorFromFloat(Value=&FloatFromCurve(FloatCurve=DataInterface() {propertyValues {Curve {keys [{arriveTangent=-3.1207, interpMode=RCIM_Cubic, leaveTangent=-3.1207, tangentMode=RCTM_User, time=0, value=1}, <at 0.5 0>]}}}))
    ScaleA=true
    ScaleRGB=true
   )
   &SubUVAnimation(
    End Frame:NiagaraInt32=&Add_Integer(A=$Particles.SubUVOffset, B=35)
    Start Frame:NiagaraInt32=&Add_Integer(A=$Particles.SubUVOffset, B=10)
   )
   &ScaleSpriteSize(Scale Factor=&Vector2DFromFloat(Value=&FloatFromCurve(FloatCurve=DataInterface() {propertyValues {Curve {keys [
          {arriveTangent=6.5992, interpMode=RCIM_Cubic, leaveTangent=2.0153, tangentMode=RCTM_User, time=0, value=0.45}
          {arriveTangent=0.671, interpMode=RCIM_Cubic, leaveTangent=0.671, tangentMode=RCTM_User, time=0.2708, value=0.77}
          {arriveTangent=0.2411, interpMode=RCIM_Cubic, leaveTangent=0.2411, tangentMode=RCTM_User, time=1, value=1}
         ]}}}))) {enabled=false}
   &DynamicMaterialParameters(Module.Param0WriteEnabled=false, Module.Param1WriteEnabled=false, Module.Param2WriteEnabled=true, Module.Param3WriteEnabled=false)
   /SetVariables(Particles.VisibilityTag=&MakeCustomIntFromBool(Bool=$User.bisHeadHidden, True INT=1, False INT=0)) {isSetParameters=true, module=SetVariables_#4}
  ]
  rendererClasses [&NiagaraSpriteRendererProperties]
  renderers [{rendererClass=&NiagaraSpriteRendererProperties}]
 }
}
summary {
 systemName=NS_Emote_FurureSamurai_Idle
 emitters [
  {emitterName=SmokeInky, rendererClasses [&NiagaraSpriteRendererProperties]}
  {emitterName=SwirlSmoke, rendererClasses [&NiagaraSpriteRendererProperties]}
  {emitterName=Darken, rendererClasses [&NiagaraSpriteRendererProperties]}
  {emitterName=SwirlSmokeLocal, rendererClasses [&NiagaraSpriteRendererProperties]}
 ]
 userVariables [
  User.bisHeadHidden {type=NiagaraBool}
  User.IsSkydiving {type=NiagaraBool}
 ]
}
